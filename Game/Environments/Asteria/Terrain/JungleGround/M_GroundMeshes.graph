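# Nanite and instancing-ready Asteria jungle ground mesh material that blends tinted floor textures into the Apollo landscape's runtime virtual texture.
# A height map blends in a glossier, lower-specular variant, then drives the RVT blend with height above the landscape and per-instance fade.
# RVT samples use blurrier mips on non-upward faces.
# Forward and mobile paths skip RVT blending, and a faint fade-scaled red emissive is added last.
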Material /Game/Environments/Asteria/Terrain/JungleGround/M_GroundMeshes
graph {
 properties {
  bTangentSpaceNormal=false
  bUseMaterialAttributes=true
  bUsedWithInstancedStaticMeshes=true
  bUsedWithNanite=true
  pixelDepthOffsetMode=PDOM_Legacy
  usedShadingModels=Default Lit
 }
 nodes [
  n1 = VertexNormalWS()
  n2 = RuntimeVirtualTextureSampleParameter!(UVs, World Position, Mip Bias=LinearInterpolate(Alpha=ComponentMask(None=@n1) {b=true}) {constA=6, constB=0}) {
   group=D_RVT
   materialType=BaseColor_Normal_Specular_YCoCg
   mipValueMode=RVTMVM_MipBias
   parameterName=RVT
   sortPriority=1
   virtualTexture=/Game/Athena/Apollo/Maps/Landscape/VT/RVT_Apollo_Landscape
  }
  n3 = SetMaterialAttributes!(
   MaterialAttributes=NamedRerouteUsage()
   Base Color=NamedRerouteDeclaration(None=@n2.BaseColor) {name=RVT Base Color, nodeColor=<rgba 0 1 0.6016 1>}
   Specular=NamedRerouteDeclaration(None=@n2.Specular) {name=RVT Specular, nodeColor=<rgba 1 0 0.8203 1>}
   Roughness=NamedRerouteDeclaration(None=@n2.Roughness) {name=RVT Rough, nodeColor=<rgba 0 1 0.7422 1>}
   Emissive Color=Constant()
   Normal=NamedRerouteDeclaration(None=@n2.Normal) {name=RVT N, nodeColor=<rgba 0 0.6953 1 1>}
  ) {attributeSetTypes [
    #1
    #2
    #3
    #4
    #5
   ]}
  n4 = ScalarParameter() {group=RVT, parameterName=Offset RVT WP, sortPriority=4}
  n5 = ComponentMask(None=WorldPosition().XYZ) {b=true}
  n6 = RuntimeVirtualTextureSampleParameter!(UVs, World Position) {
   group=D_RVT
   materialType=WorldHeight
   parameterName=RVTHeight
   sortPriority=1
   virtualTexture=/Game/Athena/Apollo/Maps/Landscape/VT/RVT_Apollo_Landscape_Height
  }
  n7 = NamedRerouteDeclaration(None=Subtract(A=Add(A=@n4, B=@n5), B=@n6.WorldHeight)) {name=RVT Height, nodeColor=<rgba 1 0 0.2344 1>}
  n8 = ScalarParameter() {
   defaultValue=5
   desc="Used to control how far colors will blend with the rvt, as well sets blending distance of normals for Up looking surfaces"
   group=RVT
   parameterName=Blend RVT
   sortPriority=2
  }
  n9 = TextureSampleParameter2D() {
   group=Textures
   parameterName=HM
   samplerType=SAMPLERTYPE_LinearGrayscale
   sortPriority=3
   texture=/Game/Environments/Asteria/Terrain/Textures/T_AsteriaJungleFloor_H
  }
  n10 = Multiply(
   A=MaterialFunctionCall(Transition Phase=Divide(A=@n7, B=@n8), Height Texture=@n9.R, Contrast=ScalarParameter() {group=RVT, parameterName=Bled RVT Contrast}) {materialFunction=&HeightLerp}.Alpha
   B=PerInstanceFadeAmount()
  )
  NamedRerouteDeclaration(None=BlendMaterialAttributes(A=@n3, B=NamedRerouteUsage(), Alpha=Saturate(None=@n10)) {vertexAttributeBlendType=UseA}) {name=RVT On, nodeColor=<rgba 0 0.8594 1 1>}
  PerInstanceRandom()
  n11 = TextureSampleParameter2D() {group=Textures, parameterName=Diffuse, sortPriority=0, texture=/Game/Environments/Asteria/Terrain/Textures/T_Asteria_JungleFloor_D}
  n12 = TextureSampleParameter2D() {
   group=Textures
   parameterName=SpecularMasks
   samplerType=SAMPLERTYPE_Masks
   sortPriority=1
   texture=/Game/Environments/Asteria/Terrain/Textures/T_Asteria_JungleFloor_S
  }
  n13 = TextureSampleParameter2D() {
   group=Textures
   parameterName=Normals
   samplerType=SAMPLERTYPE_Normal
   sortPriority=2
   texture=/Game/Environments/Asteria/Terrain/Textures/T_Asteria_JungleFloor_N
  }
  n14 = SetMaterialAttributes!(
   MaterialAttributes
   Base Color=Multiply(A=VectorParameter() {group=Extra, parameterName=Diffuse Tint, defaultValue=<rgba 1 1 1 0>}.RGB, B=@n11.RGB)
   Specular=@n12.R
   Roughness=LinearInterpolate(A=ScalarParameter() {parameterName=Roughness Min}, B=ScalarParameter() {defaultValue=1, parameterName=Roughness Max}, Alpha=@n12.B)
   Normal=Transform(None=@n13.RGB)
   Subsurface Color=Constant()
  ) {attributeSetTypes [
    #1
    #2
    #3
    #5
    #6
   ]}
  n15 = GetMaterialAttributes(None=@n14) {attributeGetTypes [#3, #2]}
  n16 = Multiply(A=@n15.Roughness) {constB=0.25}
  n17 = Multiply(A=@n15.Specular) {constB=0.2}
  n18 = VectorParameter() {group=Extra, parameterName=SSS, defaultValue=<rgba 0.0669 0.2066 0.0214 1>}
  n19 = Constant() {r=0.65}
  n20 = NamedRerouteDeclaration(None=BlendMaterialAttributes(
    A=@n14
    B=SetMaterialAttributes!(MaterialAttributes=@n15.MaterialAttributes, Roughness=@n16, Specular=@n17, Subsurface Color=@n18.RGB) {attributeSetTypes [#3, #2, #6]}
    Alpha=MaterialFunctionCall(Transition Phase=@n19, Height Texture=@n9.R) {materialFunction=&HeightLerp}.Results
   )) {name=Pre RVT, nodeColor=<rgba 1 0.7734 0 1>}
  n21 = MaterialFunctionCall(NoRVT=@n20, RVT=MaterialProxyReplace(Realtime=NamedRerouteUsage(), MaterialProxy=@n20)) {materialFunction=&FN_RVT_Switch_Material}
  n22 = PerInstanceFadeAmount()
  n23 = ScalarParameter() {defaultValue=0.0125}
  SetMaterialAttributes!(
   MaterialAttributes=ShadingPathSwitch(Default=@n21.Result, Deferred=@n21.Result, Forward=@n20, Mobile=@n20)
   Emissive Color=Multiply(A=Constant3Vector() {constant=<rgba 1 0 0 0>}, B=Multiply(A=@n22, B=@n23))
  ) {attributeSetTypes [#4]}
 ]
}
